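; Jungle fog actor that shows a fog cube with a configurable material, hidden on mobile and Switch.
; UpdateParticle, run from construction and BeginPlay, blocking-loads the assigned material into a dynamic instance if none exists.
; UpdateMaterial then pushes per-time-of-day density and color, plus noise and falloff variables, into that instance.
; Shadow quality changes show the cube only at level 2 or above, though BeginPlay skips that check.
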(event EventBeginPlay
  (bind _fogcube (Variables|Default|GetFogCube))
  (if (or (Fortnite|IsMobilePlatform) (TouchPlatform|IsNintendoSwitchorNintendoSwitchPreview))
    (Rendering|SetVisibility _fogcube false false)
    (else
      (Rendering|SetVisibility _fogcube true false)
      (CallFunction|UpdateParticle)
      (EventDispatchers|BindEventtoOnShadowQualityChanged (Settings|GetFortGameUserSettings) (delegate Custom|OnHandleShadowQualityChanged)))))

(event Custom|OnHandleShadowQualityChanged
  (Rendering|SetVisibility (Variables|Default|GetFogCube) (>= (Settings|GetShadowQuality (Settings|GetFortGameUserSettings)) 2) false))

(fn BuildExposedAssetProviderVersion(Implementation)
  (params (Version :type Fort%20Exposed%20Asset%20Provider%20Version%20Structure%20%28by%20ref%29))
  (|CallParentFunction Version)
  (ExposedAssetProvider|AddVersion Version 1)
  (return))

(fn CollectAssetReferencestoAllow(Implementation)
  (params (Collector :type Fort%20Exposed%20Asset%20Reference%20Collector%20Structure%20%28by%20ref%29))
  (|CallParentFunction Collector)
  (ExposedAssetProvider|CollectIndirectAssetReference Collector (Variables|Default|GetAssigned%20Material))
  (return))

(fn SetScalarOnFogMaterial
  (params (Value :type Float%20%28double-precision%29) (Name :type Name))
  (Rendering|Material|SetScalarParameterValue (Variables|Default|GetFogMaterial) Name Value :!class /Script/Engine.MaterialInstanceDynamic))

(fn SetVectorOnFogMaterial
  (params (Value :type Linear%20Color%20Structure) (Name :type Name))
  (Rendering|Material|SetVectorParameterValue (Variables|Default|GetFogMaterial) Name Value :!class /Script/Engine.MaterialInstanceDynamic))

(fn UpdateMaterial
  (Utilities|FlowControl|Sequence
    (:then_0
      ;[1 1 1] Density
      (CallFunction|SetScalaronFogMaterial (Variables|Density|GetMorning_Density) "Opacity_Morning")
      (CallFunction|SetScalaronFogMaterial (Variables|Density|GetDay_Density) "Opacity_Day")
      (CallFunction|SetScalaronFogMaterial (Variables|Density|GetEvening_Density) "Opacity_Evening")
      (CallFunction|SetScalaronFogMaterial (Variables|Density|GetNight_Density) "Opacity_Night")
      (CallFunction|SetScalaronFogMaterial (Variables|Density|GetMorning_EmissiveDensity) "Emissiveopacity_Morning")
      (CallFunction|SetScalaronFogMaterial (Variables|Density|GetDay_EmissiveDensity) "Emissiveopacity_Day")
      (CallFunction|SetScalaronFogMaterial (Variables|Density|GetEvening_EmissiveDensity) "Emissiveopacity_Evening")
      (CallFunction|SetScalaronFogMaterial (Variables|Density|GetNight_EmissiveDensity) "Emissiveopacity_Night"))
    (:then_1
      (CallFunction|SetScalaronFogMaterial (Variables|Density|GetNearfade_Distance) "NearFade_Distance")
      (CallFunction|SetScalaronFogMaterial (Variables|Density|GetNearfade_Exponent) "NearFade_Exponent")
      (CallFunction|SetScalaronFogMaterial (Variables|Density|GetHeightFalloff_Low) "HeightFalloff_Low")
      (CallFunction|SetScalaronFogMaterial (Variables|Density|GetHeightFalloff_Mid) "HeightFalloff_Mid")
      (CallFunction|SetScalaronFogMaterial (Variables|Density|GetHeightFalloff_High) "HeightFalloff_High")
      (CallFunction|SetScalaronFogMaterial (Variables|Density|GetHeightFalloff_Exponent) "HeightFalloff_Exponent"))
    (:then_2
      ;[1 1 1] Color Setup
      (CallFunction|SetVectoronFogMaterial (Variables|Color|Albedo|GetA%20Morning) "Fog_Albedo_Color_Morning")
      (CallFunction|SetVectoronFogMaterial (Variables|Color|Albedo|GetA%20Day) "Fog_Albedo_Color_Day")
      (CallFunction|SetVectoronFogMaterial (Variables|Color|Albedo|GetA%20Evening) "Fog_Albedo_Color_Evening")
      (CallFunction|SetVectoronFogMaterial (Variables|Color|Albedo|GetA%20Night) "Fog_Albedo_Color_Night")
      (CallFunction|SetVectoronFogMaterial (Variables|Color|Emissive|GetE%20Morning) "Fog_Emissive_Color_Morning")
      (CallFunction|SetVectoronFogMaterial (Variables|Color|Emissive|GetE%20Day) "Fog_Emissive_Color_Day")
      (CallFunction|SetVectoronFogMaterial (Variables|Color|Emissive|GetE%20Evening) "Fog_Emissive_Color_Evening")
      (CallFunction|SetVectoronFogMaterial (Variables|Color|Emissive|GetE%20Night) "Fog_Emissive_Color_Night")
      (CallFunction|SetScalaronFogMaterial (Variables|Color|GetGround%20Atmosphere%20Color%20Blend) "UseGroundAtmosphereColor"))
    (:then_3
      ;[1 1 1] Noise
      (CallFunction|SetScalaronFogMaterial (Variables|Noise|GetMacro_NoiseScale) "MacroNoiseScale")
      (CallFunction|SetScalaronFogMaterial (Variables|Noise|GetCurlNoiseScale) "CurlNoiseScale")
      (CallFunction|SetScalaronFogMaterial (Variables|Noise|GetCurlNoiseIntensity) "CurlNoiseIntensity")
      (CallFunction|SetVectoronFogMaterial (Math|Conversions|ToLinearColor(Vector) (Variables|Noise|GetCurlNoisePanSpeedAndDirection)) "CurlNoiseMotion")
      (CallFunction|SetVectoronFogMaterial (Math|Conversions|ToLinearColor(Vector) (Variables|Noise|GetFogNoisePanSpeed)) "MacroNoiseMotion")
      (CallFunction|SetVectoronFogMaterial 0 (Math|Color|MakeColor (Variables|Noise|GetBlackValue) (Variables|Noise|GetMiddleValue) (Variables|Noise|GetWhiteValue) (Variables|Noise|GetMiddlePoint)) "Levels"))
    (:then_4
      ;[1 1 1] Falloff
      (CallFunction|SetScalaronFogMaterial (Variables|Density|GetBoxFalloffExponent_X) "BoxFalloffExponent_X")
      (CallFunction|SetScalaronFogMaterial (Variables|Density|GetBoxFalloffExponent_Y) "BoxFalloffExponent_Y")
      (CallFunction|SetScalaronFogMaterial (Variables|Density|GetSphereFalloffHardness) "SphereFalloffHardness"))))

(fn updateParticle
  (bind _fogcube (Variables|Default|GetFogCube))
  (Utilities|FlowControl|Sequence
    (:then_0
      (if (TouchPlatform|IsMobileGame)
        (Rendering|SetVisibility _fogcube false false)
        (else
          (Rendering|SetVisibility _fogcube true false)
          (Utilities|IsValid (Variables|Default|GetFogMaterial)
            (:Is%20Not%20Valid
              (bind _returnvalue (Utilities|LoadAssetBlocking (Variables|Default|GetAssigned%20Material)))
              (bind _asmaterial_interface (Utilities|Casting|CastToMaterialInterface _returnvalue))
              (bind _returnvalue_2 (Rendering|Material|CreateDynamicMaterialInstance _asmaterial_interface (Utilities|String|StringToName (Utilities|String|Append (Utilities|GetDisplayName self) (Utilities|String|ToString(Integer) (Math|Float|Truncate (/ (Math|Float|float+float (.x (Transformation|GetActorLocation self)) (.y (Transformation|GetActorLocation self)) (.z (Transformation|GetActorLocation self))) 3.0))))) "None" :!class /Script/Engine.KismetMaterialLibrary))
              (Variables|Default|SetFogMaterial _returnvalue_2)
              (Rendering|Material|SetMaterial (Variables|Default|GetFogCube) 0 (Variables|Default|GetFogMaterial))
              (CallFunction|UpdateMaterial))))))
    (:then_1
      (Utilities|IsValid (Variables|Default|GetFogMaterial)
        (:Is%20Valid
          (Development|SetHiddeninGame (Variables|Default|GetFogCube) false false))))))

(fn ConstructionScript
  (CallFunction|UpdateParticle))
