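# Additive volume-domain material for jungle fog on objects or Niagara particles, with time-of-day albedo, emissive and opacity.
# World-space noise, either curl-distorted Voronoi or projected Perlin, is leveled, becoming constant density on low and medium quality.
# Sphere or bounding-box falloff with height gradient shapes extinction, scaled by camera near fade and a global RainForest opacity.
# Switches add view-angle density and a baked density map; colors can blend toward atmospheric light.
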Material /Game/Athena/Asteria/Environments/Landscape/Materials/DistanceFog/M_JungleVolumeFog
graph {
 properties {
  bUsedWithNiagaraSprites=true
  blendMode=BLEND_Additive
  materialDomain=MD_Volume
  pixelDepthOffsetMode=PDOM_Legacy
  usedShadingModels=Default Lit
 }
 nodes [
  Composite() {desc=TOD Controlled, subgraphName=Fog_Emissive_Color}
  Composite() {desc=TOD Controlled, subgraphName=Fog_Albedo_Color}
  n1 = MaterialFunctionCall(
   Color_Morning=VectorParameter() {parameterName=Fog_Albedo_Color_Morning, defaultValue=<rgba 1 1 1 1>}.RGB
   Color_Day=VectorParameter() {parameterName=Fog_Albedo_Color_Day, defaultValue=<rgba 1 1 1 1>}.RGB
   Color_Evening=VectorParameter() {parameterName=Fog_Albedo_Color_Evening, defaultValue=<rgba 1 1 1 1>}.RGB
   Color_Night=VectorParameter() {parameterName=Fog_Albedo_Color_Night, defaultValue=<rgba 1 1 1 1>}.RGB
  ) {materialFunction=&Color_Time_of_Day}
  MaterialExpressionReroute_1 = Reroute(None=@n1.Result)
  Multiply(
   A=LinearInterpolate(A=@MaterialExpressionReroute_1.Result, B=NamedRerouteUsage(), Alpha=ScalarParameter() {parameterName=UseGroundAtmosphereColor, sliderMax=1})
   B=NamedRerouteUsage()
  )
  n2 = MaterialFunctionCall(
   Color_Morning=VectorParameter() {parameterName=Fog_Emissive_Color_Morning, defaultValue=<rgba 1 1 1 1>}.RGB
   Color_Day=VectorParameter() {parameterName=Fog_Emissive_Color_Day, defaultValue=<rgba 1 1 1 1>}.RGB
   Color_Evening=VectorParameter() {parameterName=Fog_Emissive_Color_Evening, defaultValue=<rgba 1 1 1 1>}.RGB
   Color_Night=VectorParameter() {parameterName=Fog_Emissive_Color_Night, defaultValue=<rgba 1 1 1 1>}.RGB
  ) {materialFunction=&Color_Time_of_Day}
  MaterialExpressionReroute_0 = Reroute(None=@n2.Result)
  n3 = LinearInterpolate(A=@MaterialExpressionReroute_0.Result, B=NamedRerouteUsage(), Alpha=ScalarParameter() {parameterName=UseGroundAtmosphereColor, sliderMax=1})
  n4 = NamedRerouteUsage()
  n5 = ScalarParameter() {defaultValue=0.1, parameterName=Emissiveopacity_Morning}
  n6 = ScalarParameter() {defaultValue=0.1, parameterName=Emissiveopacity_Day}
  n7 = ScalarParameter() {defaultValue=0.1, parameterName=Emissiveopacity_Evening}
  n8 = ScalarParameter() {defaultValue=0.1, parameterName=Emissiveopacity_Night}
  n9 = Distance(A=WorldPosition().XYZ, B=CameraPositionWS())
  n10 = ScalarParameter() {defaultValue=1000, parameterName=NearFade_Distance}
  n11 = Power(Base=Saturate(None=Divide(A=@n9, B=@n10)), Exp=ScalarParameter() {defaultValue=1, parameterName=NearFade_Exponent}) {constExponent=16}
  Multiply(
   A=Multiply(
    A=Multiply(A=@n3, B=@n4)
    B=MaterialFunctionCall(Color_Morning=@n5, Color_Day=@n6, Color_Evening=@n7, Color_Night=@n8) {materialFunction=&Color_Time_of_Day}.Result
   )
   B=@n11
  )
  n12 = Divide(A=WorldPosition().XYZ, B=ScalarParameter() {defaultValue=100, parameterName=MacroNoiseScale})
  n13 = Divide(A=WorldPosition().XYZ, B=ScalarParameter() {defaultValue=1024, parameterName=CurlNoiseScale})
  n14 = Multiply(
   A=Multiply(A=Time(), B=VectorParameter() {parameterName=CurlNoiseMotion, defaultValue=<rgba -0.02 -0.02 -0.02 1>}.RGB)
   B=ScalarParameter() {defaultValue=0.25, parameterName=CurlNoiseSpeed}
  )
  n15 = Multiply(
   A=TextureSample(UVs=Add(A=@n13, B=@n14)) {samplerType=SAMPLERTYPE_LinearColor, texture=/Game/Effects/FogVolume/VT_CurlNoise_LOW}.RGB
   B=ScalarParameter() {defaultValue=0.5, parameterName=CurlNoiseIntensity}
  )
  n16 = Time()
  n17 = VectorParameter() {parameterName=MacroNoiseMotion, defaultValue=<rgba 0.01 0.015 0.015 1>}
  n18 = TextureSample(UVs=Add(A=Add(A=@n12, B=@n15), B=Multiply(A=@n16, B=@n17.RGB))) {samplerType=SAMPLERTYPE_LinearColor, texture=/Game/Effects/FogVolume/VT_Voronoi_LOW}
  n19 = VectorParameter() {parameterName=Albedo_Noise_Scale, channelNames {a="", b=Z_Scale, g=Y_Scale, r=X_Scale}, defaultValue=<rgba 512 512 512 1>}
  n20 = VectorParameter() {parameterName=Albedo_Noise_Panspeed, channelNames {a="", b=Z_Speed, g=Y_Speed, r=X_Speed}, defaultValue=<rgba 0.02 0.02 0.02 1>}
  n21 = Add(A=Divide(A=WorldPosition().XYZ, B=@n19), B=Multiply(A=Time(), B=@n20))
  n22 = ComponentMask(None=@n21) {g=true, r=true}
  n23 = ScalarParameter() {defaultValue=4, parameterName=AlbedoFogVerticalDivisor}
  n24 = Divide(A=ComponentMask(None=@n21) {b=true}, B=@n23) {constB=4}
  n25 = ComponentMask(None=@n21) {b=true, g=true}
  n26 = Divide(A=ComponentMask(None=@n21) {r=true}, B=@n23) {constB=4}
  n27 = Min(
   A=TextureSampleParameter2D(UVs=Add(A=@n22, B=@n24)) {parameterName=AlbedoNoise, samplerType=SAMPLERTYPE_LinearColor, texture=/Game/Effects/SnowSandInteraction/Textures/SSI_Perlin}.RGB
   B=TextureSampleParameter2D(UVs=Add(A=@n25, B=@n26)) {parameterName=AlbedoNoise, samplerType=SAMPLERTYPE_LinearColor, texture=&SSI_Perlin}.RGB
  )
  n28 = ScalarParameter() {defaultValue=1, parameterName=AlbedoNoisePower}
  n29 = ComponentMask(None=StaticSwitchParameter(True=@n18.R, False=Power(Base=@n27, Exp=@n28)) {defaultValue=true, parameterName=UseCurlNoise}) {r=true}
  n30 = VectorParameter() {parameterName=Levels, channelNames {a=Middlepoint, b=WhiteValue, g=MiddleValue, r=BlackValue}, defaultValue=<rgba 0 0.25 1 0.5>}
  n31 = ScalarParameter() {defaultValue=0.6, parameterName=LowQualityDensityFactor}
  NamedRerouteDeclaration(None=QualitySwitch(
    Default=MaterialFunctionCall!(
     Texture=@n29
     ---------------
     New Black Value=@n30.BlackValue
     New Middle Value=@n30.BlackValue
     New White Value=@n30.BlackValue
     Middle Point=@n30.BlackValue
     "--------------- "
     Define Interpolation Curve
     Interpolation Power
     Invert Interpolation Power
    ) {materialFunction=/Engine/Functions/Engine_MaterialFunctions02/3PointLevels}.Result
    Low=@n31
    Medium=@n31
   )) {name=ProjectedNoise, nodeColor=<rgba 1 0.75 0 1>}
  NamedRerouteDeclaration(None=Divide(A=AtmosphericLightColor(), B=ScalarParameter() {defaultValue=5, parameterName=Atmosphere remap})) {name=Atmosphere Color, nodeColor=<rgba 0.4531 0 1 1>}
  PinBase()
  PinBase!(Result=@n2.Result) {reroutePins [Result {expression=&self:MaterialExpressionReroute_0}]}
  PinBase!()
  PinBase!(Result=@n1.Result) {reroutePins [Result {expression=&self:MaterialExpressionReroute_1}]}
  n32 = MaterialFunctionCall!(
   Density Map=TextureObject() {samplerType=SAMPLERTYPE_LinearColor, texture=/Game/ArtTools/Tech_Art/FogDensityBaker/T_FogDensity}
   Min Bounds=Constant2Vector() {desc=Min Bounds, g=-96628.9766, r=-34583.8008}
   Max Bounds=Constant2Vector() {desc=Max Bounds, g=48084.2344, r=110129.4141}
   Min Density
   Max Density
  ) {materialFunction=/Game/ArtTools/Tech_Art/FogDensityBaker/MF_FogDensityMapping}
  n33 = ComponentMask(None=MaterialFunctionCall() {materialFunction=&BoundingBoxBased_0-1_UVW}.RGB) {g=true, r=true}
  n34 = MaterialFunctionCall(Float2=Abs(None=ConstantBiasScale(None=@n33) {bias=-0.5, scale=2})) {materialFunction=&BreakOutFloat2Components}
  n35 = ScalarParameter() {defaultValue=2, parameterName=BoxFalloffExponent_X}
  n36 = ScalarParameter() {defaultValue=2, parameterName=BoxFalloffExponent_Y}
  n37 = Multiply(A=OneMinus(None=Power(Base=@n34.R, Exp=@n35) {constExponent=4}), B=OneMinus(None=Power(Base=@n34.G, Exp=@n36) {constExponent=4}))
  n38 = ParticleRadius()
  n39 = Multiply(A=StaticSwitchParameter(True=Multiply(A=@n38) {constB=2}, False=ObjectRadius()) {parameterName=IsParticle}) {constB=0.8}
  n40 = ScalarParameter() {defaultValue=1, parameterName=RadiusDivider1}
  n41 = LinearInterpolate(
   A=@n40
   B=ScalarParameter() {defaultValue=1.5, parameterName=RadiusDivider2}
   Alpha=Power(Base=@n18.R, Exp=ScalarParameter() {defaultValue=0.3, parameterName=RadiusVariationExponent})
  )
  n42 = SphereMask(
   A=StaticSwitchParameter(True=ParticlePositionWS(), False=ObjectPositionWS()) {parameterName=IsParticle}
   B=WorldPosition().XYZ
   Radius=QualitySwitch(Default=Divide(A=@n39, B=@n41), Low=@n40, Medium=@n40)
   Hardness=ScalarParameter() {defaultValue=0.5, parameterName=SphereFalloffHardness}
  ) {hardnessPercent=50}
  n43 = MaterialFunctionCall(
   1=ScalarParameter() {defaultValue=1, parameterName=HeightFalloff_Low}
   2=ScalarParameter() {defaultValue=1, parameterName=HeightFalloff_Mid}
   3=ScalarParameter() {defaultValue=1, parameterName=HeightFalloff_High}
   A=MaterialFunctionCall() {materialFunction=&BoundingBoxBased_0-1_UVW}.B
  ) {materialFunction=&Lerp_Multiple_Float3}
  n44 = ScalarParameter() {defaultValue=1, parameterName=HeightFalloff_Exponent}
  n45 = Multiply(
   A=NamedRerouteUsage()
   B=Multiply(A=StaticSwitchParameter(True=@n37, False=@n42) {parameterName=UseBoundingBoxFalloff}, B=Power(Base=@n43.Lerp 3 Inputs, Exp=@n44)) {constA=1}
  )
  n46 = ScalarParameter() {defaultValue=0.5, parameterName=Opacity_Morning}
  n47 = ScalarParameter() {defaultValue=0.5, parameterName=Opacity_Day}
  n48 = ScalarParameter() {defaultValue=0.5, parameterName=Opacity_Evening}
  n49 = ScalarParameter() {defaultValue=0.5, parameterName=Opacity_Night}
  n50 = Multiply(
   A=@n11
   B=Multiply(
    A=NamedRerouteDeclaration(None=@n45) {name=Extinction Mask, nodeColor=<rgba 0 0.625 1 1>}
    B=MaterialFunctionCall(Color_Morning=@n46, Color_Day=@n47, Color_Evening=@n48, Color_Night=@n49) {materialFunction=&Color_Time_of_Day}.Result
   ) {constA=1}
  )
  n51 = Normalize(VectorInput=Subtract(A=WorldPosition().XYZ, B=CameraPositionWS()))
  n52 = Constant3Vector() {constant=<rgba 0 0 -1 0>}
  n53 = Max(A=Saturate(None=DotProduct(A=@n51, B=@n52)), B=ScalarParameter() {defaultValue=0.1, parameterName=MinAngleBasedDensity})
  n54 = CollectionParameter() {
   collection=&FortniteMaterialParameters
   parameterName=RainForest_Fog_MasterOpacity
  }
  n55 = Multiply(A=StaticSwitchParameter(True=Multiply(A=@n50, B=@n53) {constB=0.5}, False=@n50) {parameterName=UseAngleBasedDensity}, B=@n54)
  QualitySwitch(Default=StaticSwitchParameter(True=Multiply(A=@n32.Result, B=@n55), False=@n55) {parameterName=UseDensityTexture}, Low=@n55)
 ]
}
